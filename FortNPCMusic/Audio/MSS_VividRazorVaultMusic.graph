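# Stereo MetaSound for NPC vault music (VividRazor), moving through three looping stages, transition stingers and an ending that finishes the source.
# A phase-driven metronome supplies bar triggers that quantize incoming gameplay events, with per-transition bar counts delaying the next loop.
# Stages 2 and 3 crossfade regular and boss stems over drum loops, and play death stingers on boss elimination.
# An eight-input mixer feeds master volume, and CurrentPhase reports the active phase.
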MetaSoundSource /FortNPCMusic/Audio/MSS_VividRazorVaultMusic
inputs [
 bIsBossGoingToSpawn {type=Bool}
 bIsBossPresent {type=Bool}
 EndingDelay {type=Int32}
 MasterVolume {type=Float}
 Mixing.BossDeathStinger {type=Float}
 Mixing.Ending {type=Float}
 Mixing.Stage1 {type=Float}
 Mixing.Stage2 {type=Float}
 Mixing.Stage3 {type=Float}
 Mixing.TransitionToPhase2 {type=Float}
 Mixing.TransitionToPhase3 {type=Float}
 NPC.Event.BossAppeared {type=Trigger}
 NPC.Event.BossEliminated {type=Trigger}
 NPC.Event.PlayEnding {type=Trigger}
 NPC.Event.PlayStage1Loop {type=Trigger}
 NPC.Event.PlayStage2Loop {type=Trigger}
 NPC.Event.PlayStage3Loop {type=Trigger}
 NPC.Event.PlayTransitionToStage1 {type=Trigger}
 NPC.Event.PlayTransitionToStage2 {type=Trigger}
 NPC.Event.PlayTransitionToStage3 {type=Trigger}
 TransitionLengthsBars {type=Int32:Array}
 UE.Source.OnPlay {type=Trigger}
 Wave.Stage1.DrumLoop {type=WaveAsset}
 Wave.Stage1.ElseLoop {type=WaveAsset}
 Wave.Stage2.Boss.ElseLoop {type=WaveAsset}
 Wave.Stage2.BossDeath {type=WaveAsset}
 Wave.Stage2.DrumLoop {type=WaveAsset}
 Wave.Stage2.ElseLoop {type=WaveAsset}
 Wave.Stage3.Boss.ElseLoop {type=WaveAsset}
 Wave.Stage3.BossDeath {type=WaveAsset}
 Wave.Stage3.DrumLoop {type=WaveAsset}
 Wave.Stage3.ElseLoop {type=WaveAsset}
 Wave.Stingers.Ending {type=WaveAsset}
 Wave.Stingers.TransitionToStage2 {type=WaveAsset}
 Wave.Stingers.TransitionToStage3 {type=WaveAsset}
]
interfaces [UE.OutputFormat.Stereo:1.0, UE.Source.OneShot:1.0, UE.Source:1.1]
outputs [
 CurrentPhase {type=Int32}
 "UE.OutputFormat.Stereo.Audio:0" {type=Audio}
 "UE.OutputFormat.Stereo.Audio:1" {type=Audio}
 UE.Source.OneShot.OnFinished {type=Trigger}
]
pages [Default {
  input_defaults {
   EndingDelay=1
   MasterVolume=1
   Mixing.BossDeathStinger=1.4
   Mixing.Ending=1
   Mixing.Stage1=1
   Mixing.Stage2=1
   Mixing.Stage3=1
   Mixing.TransitionToPhase2=1
   Mixing.TransitionToPhase3=1
   NPC.Event.BossAppeared=null
   NPC.Event.BossEliminated=null
   NPC.Event.PlayEnding=null
   NPC.Event.PlayStage1Loop=null
   NPC.Event.PlayStage2Loop=null
   NPC.Event.PlayStage3Loop=null
   NPC.Event.PlayTransitionToStage1=null
   NPC.Event.PlayTransitionToStage2=null
   NPC.Event.PlayTransitionToStage3=null
   UE.Source.OnPlay=null
   Wave.Stage1.DrumLoop=null
   Wave.Stage1.ElseLoop=null
   Wave.Stage2.Boss.ElseLoop=null
   Wave.Stage2.BossDeath=null
   Wave.Stage2.DrumLoop=null
   Wave.Stage2.ElseLoop=null
   Wave.Stage3.Boss.ElseLoop=null
   Wave.Stage3.BossDeath=null
   Wave.Stage3.DrumLoop=null
   Wave.Stage3.ElseLoop=null
   Wave.Stingers.Ending=null
   Wave.Stingers.TransitionToStage2=null
   Wave.Stingers.TransitionToStage3=null
   bIsBossGoingToSpawn="<MetaSoundBuilderResult.SUCCEEDED: 0>"
   bIsBossPresent=null
   TransitionLengthsBars [4,4]
  }
  nodes [
   InitVariable.Float()
   InitVariable.Float()
   InitVariable.Float()
   InitVariable.Float()
   InitVariable.Int32()
   InitVariable.Trigger()
   n1 = InitVariable.Float()
   n2 = InitVariable.Trigger()
   n3 = VariableDeferredAccessor.Trigger(Variable=@n2.Variable)
   TriggerAnyTriggerAny2_3 = "TriggerAny.Trigger Any (2)"(In 0=^UE.Source.OnPlay, In 1=@n3.Value)
   n4 = VariableDeferredAccessor.Trigger(Variable=InitVariable.Trigger().Variable)
   n5 = VariableDeferredAccessor.Trigger(Variable=InitVariable.Trigger().Variable)
   n6 = VariableDeferredAccessor.Trigger(Variable=InitVariable.Trigger().Variable)
   n7 = VariableDeferredAccessor.Trigger(Variable=InitVariable.Trigger().Variable)
   TriggerRouteTriggerRouteInt324_0 = "TriggerRoute.Trigger Route (Int32, 4):Int32"(Set 3=@n4.Value, Set 1=@n5.Value, Set 2=@n6.Value, Set 0=@n7.Value) {also {defaults {Value 0=0, Value 1=1, Value 2=2, Value 3=3}}}
   n8 = VariableAccessor.Int32(Variable=VariableMutator.Int32(Variable=InitVariable.Int32().Variable, Value=@TriggerRouteTriggerRouteInt324_0.Value).Variable)
   n9 = /FortMetasound/Patches/MSP_SwitchOnInt_4(Compare=@TriggerAnyTriggerAny2_3.Out, Int=@n8.Value)
   TriggerRouteTriggerRouteFloat4_0 = "TriggerRoute.Trigger Route (Float, 4):Float"(Set 0=@n9.Out 0, Set 1=@n9.Out 1, Set 2=@n9.Out 2, Set 3=@n9.Out 3) {also {defaults {Value 0=142, Value 1=142, Value 2=150, Value 3=150}}}
   n10 = VariableAccessor.Float(Variable=VariableMutator.Float(Variable=InitVariable.Float().Variable, Value=@TriggerRouteTriggerRouteFloat4_0.Value).Variable)
   TriggerCounter_0 = UE.Trigger Counter(In=/FortNPCMusic/Audio/Patches/MSP_NPCMusic_Metronome(UE.Source.OnPlay=^UE.Source.OnPlay, BPM=@n10.Value).OnBeat) {also {defaults {Reset Count=4}}}
   n11 = UE.Modulo:Int32(PrimaryOperand=@TriggerCounter_0.Count) {also {defaults {AdditionalOperands=4}}}
   n12 = VariableAccessor.Trigger(Variable=VariableMutator.Trigger(
     Variable=InitVariable.Trigger().Variable
     Value=TriggerCompare.Clamp:Int32(A=@n11.Out, Compare=@TriggerCounter_0.On Trigger) {also {defaults {B=1}}}.True
    ).Variable)
   n13 = VariableAccessor.Trigger(Variable=@n12.Variable)
   n14 = VariableAccessor.Trigger(Variable=@n13.Variable)
   n15 = VariableAccessor.Trigger(Variable=@n14.Variable)
   n16 = VariableAccessor.Trigger(Variable=@n15.Variable)
   n17 = VariableAccessor.Trigger(Variable=@n16.Variable)
   n18 = VariableAccessor.Trigger(Variable=@n17.Variable)
   n19 = &MSP_ScheduleTrigger(Schedule=^NPC.Event.BossEliminated, In Trigger=@n18.Value)
   n20 = VariableAccessor.Trigger(Variable=VariableMutator.Trigger(Variable=InitVariable.Trigger().Variable, Value=@n19.Scheduled Trigger).Variable)
   n21 = &MSP_ScheduleTrigger(Schedule=^NPC.Event.PlayTransitionToStage3, In Trigger=@n14.Value)
   n22 = VariableAccessor.Trigger(Variable=VariableMutator.Trigger(Variable=InitVariable.Trigger().Variable, Value=@n21.Scheduled Trigger).Variable)
   TriggerAnyTriggerAny2_2 = "TriggerAny.Trigger Any (2)"(In 0=^NPC.Event.PlayStage3Loop, In 1=@n22.Value)
   n23 = VariableAccessor.Int32(Variable=@n8.Variable)
   n24 = VariableDeferredAccessor.Trigger(Variable=@n2.Variable)
   n25 = VariableAccessor.Int32(Variable=VariableMutator.Int32(
     Value=Array.Get:Int32:Array(Array=^TransitionLengthsBars, Index=@n23.Value, Trigger=@n24.Value).Element
     Variable=InitVariable.Int32().Variable
    ).Variable)
   TriggerRouteTriggerRouteInt322_0 = "TriggerRoute.Trigger Route (Int32, 2):Int32"(Set 0=^NPC.Event.PlayStage3Loop, Value 1=VariableAccessor.Int32(Variable=@n25.Variable).Value, Set 1=@n22.Value) {also {defaults {Value 0=0}}}
   n26 = VariableAccessor.Trigger(Variable=VariableMutator.Trigger(
     Value=&MSP_ScheduleTrigger(Schedule=@TriggerAnyTriggerAny2_2.Out, Triggers From Now=@TriggerRouteTriggerRouteInt322_0.Value, In Trigger=@n15.Value).Scheduled Trigger
     Variable=@n6.Variable
    ).Variable)
   n27 = VariableAccessor.Trigger(Variable=@n26.Variable)
   n28 = &MSP_ScheduleTrigger(Triggers From Now=^EndingDelay, Schedule=^NPC.Event.PlayEnding, In Trigger=@n16.Value)
   n29 = VariableAccessor.Trigger(Variable=VariableMutator.Trigger(Variable=InitVariable.Trigger().Variable, Value=@n28.Scheduled Trigger).Variable)
   n30 = VariableAccessor.Trigger(Variable=@n29.Variable)
   n31 = VariableAccessor.Trigger(Variable=@n30.Variable)
   TriggerAnyTriggerAny2_4 = "TriggerAny.Trigger Any (2)"(In 0=^NPC.Event.PlayStage1Loop, In 1=^NPC.Event.PlayTransitionToStage1)
   n32 = VariableAccessor.Trigger(Variable=@n18.Variable)
   n33 = VariableAccessor.Trigger(Variable=@n32.Variable)
   n34 = VariableAccessor.Trigger(Variable=VariableMutator.Trigger(
     Value=&MSP_ScheduleTrigger(Schedule=@TriggerAnyTriggerAny2_4.Out, In Trigger=@n33.Value).Scheduled Trigger
     Variable=@n7.Variable
    ).Variable)
   n35 = VariableAccessor.Trigger(Variable=@n34.Variable)
   n36 = &MSP_ScheduleTrigger(Schedule=^NPC.Event.PlayTransitionToStage2, In Trigger=@n12.Value)
   n37 = VariableAccessor.Trigger(Variable=VariableMutator.Trigger(Variable=InitVariable.Trigger().Variable, Value=@n36.Scheduled Trigger).Variable)
   TriggerAnyTriggerAny2_5 = "TriggerAny.Trigger Any (2)"(In 0=^NPC.Event.PlayStage2Loop, In 1=@n37.Value)
   TriggerRouteTriggerRouteInt322_1 = "TriggerRoute.Trigger Route (Int32, 2):Int32"(Set 0=^NPC.Event.PlayStage2Loop, Value 1=@n25.Value, Set 1=@n37.Value)
   n38 = VariableAccessor.Trigger(Variable=VariableMutator.Trigger(
     Value=&MSP_ScheduleTrigger(Schedule=@TriggerAnyTriggerAny2_5.Out, Triggers From Now=@TriggerRouteTriggerRouteInt322_1.Value, In Trigger=@n13.Value).Scheduled Trigger
     Variable=@n5.Variable
    ).Variable)
   n39 = VariableAccessor.Trigger(Variable=@n38.Variable)
   n40 = VariableAccessor.Trigger(Variable=@n39.Variable)
   TriggerAnyTriggerAny4_2 = "TriggerAny.Trigger Any (4)"(In 2=@n27.Value, In 3=@n31.Value, In 0=@n35.Value, In 1=@n40.Value)
   n41 = TriggerCompare.Clamp:Bool(
    A=^bIsBossPresent
    Compare=VariableAccessor.Trigger(Variable=VariableMutator.Trigger(Value=@TriggerAnyTriggerAny4_2.Out, Variable=@n3.Variable).Variable).Value
   ) {also {defaults {B="<MetaSoundBuilderResult.SUCCEEDED: 0>"}}}
   TriggerAnyTriggerAny2_0 = "TriggerAny.Trigger Any (2)"(In 0=&MSP_ScheduleTrigger(Schedule=^NPC.Event.BossAppeared, In Trigger=@n17.Value).Scheduled Trigger, In 1=@n41.True)
   n42 = VariableAccessor.Trigger(Variable=VariableMutator.Trigger(Variable=InitVariable.Trigger().Variable, Value=@TriggerAnyTriggerAny2_0.Out).Variable)
   n43 = VariableDeferredAccessor.Trigger(Variable=InitVariable.Trigger().Variable)
   n44 = VariableDeferredAccessor.Trigger(Variable=InitVariable.Trigger().Variable)
   TriggerAnyTriggerAny2_6 = "TriggerAny.Trigger Any (2)"(In 0=@n43.Value, In 1=@n44.Value)
   TriggerAnyTriggerAny2_1 = "TriggerAny.Trigger Any (2)"(
    In 0=&MSP_ScheduleTrigger(Schedule=@TriggerAnyTriggerAny2_6.Out, In Trigger=@n32.Value).Scheduled Trigger
    In 1=@n41.False
   )
   n45 = VariableAccessor.Trigger(Variable=VariableMutator.Trigger(Variable=InitVariable.Trigger().Variable, Value=@TriggerAnyTriggerAny2_1.Out).Variable)
   TriggerRouteTriggerRouteFloat3_0 = "TriggerRoute.Trigger Route (Float, 3):Float"(Set 2=@n20.Value, Set 1=@n42.Value, Set 0=@n45.Value) {also {defaults {Value 0=1, Value 2=0}}}
   n46 = UE.InterpTo:Audio(Target=@TriggerRouteTriggerRouteFloat3_0.Value) {also {defaults {Interp Time=0.5}}}
   n47 = VariableAccessor.Trigger(Variable=@n35.Variable)
   n48 = VariableAccessor.Trigger(Variable=@n40.Variable)
   n49 = VariableAccessor.Trigger(Variable=@n48.Variable)
   TriggerAnyTriggerAny4_1 = "TriggerAny.Trigger Any (4)"(In 0=VariableAccessor.Trigger(Variable=@n47.Variable).Value, In 2=@n49.Value, In 1=@n39.Value, In 3=@n30.Value)
   n50 = VariableAccessor.Trigger(Variable=VariableMutator.Trigger(Variable=InitVariable.Trigger().Variable, Value=@TriggerAnyTriggerAny4_1.Out).Variable)
   n51 = /FortNPCMusic/Audio/Patches/MSP_LayeredTrackPlayback2(
    Stem1.Stem=^Wave.Stage3.DrumLoop
    Stem2.Stem=^Wave.Stage3.ElseLoop
    Stem2.Volume=VariableAccessor.Float(Variable=VariableMutator.Float(Variable=@n1.Variable, Value=@n46.Value).Variable).Value
    Play=@n26.Value
    Stop=@n50.Value
   )
   TriggerRouteTriggerRouteFloat3_1 = "TriggerRoute.Trigger Route (Float, 3):Float"(
    Set 0=VariableAccessor.Trigger(Variable=@n45.Variable).Value
    Set 1=VariableAccessor.Trigger(Variable=@n42.Variable).Value
    Set 2=VariableAccessor.Trigger(Variable=@n20.Variable).Value
   ) {also {defaults {Value 1=1, Value 2=0}}}
   n52 = UE.InterpTo:Audio(Target=@TriggerRouteTriggerRouteFloat3_1.Value) {also {defaults {Interp Time=0.5}}}
   n53 = VariableAccessor.Float(Variable=VariableMutator.Float(Variable=InitVariable.Float().Variable, Value=@n52.Value).Variable)
   n54 = VariableAccessor.Trigger(Variable=@n27.Variable)
   n55 = VariableAccessor.Trigger(Variable=@n54.Variable)
   WavePlayer_5 = UE.Wave Player:Stereo(
    Wave Asset=^Wave.Stage3.Boss.ElseLoop
    Stop=VariableAccessor.Trigger(Variable=@n50.Variable).Value
    Play=VariableAccessor.Trigger(Variable=@n55.Variable).Value
   ) {also {defaults {Loop="<MetaSoundBuilderResult.SUCCEEDED: 0>"}}}
   n56 = InitVariable.Float()
   n57 = VariableAccessor.Trigger(Variable=VariableMutator.Trigger(Variable=InitVariable.Trigger().Variable, Value=@n21.Scheduled Trigger).Variable)
   TriggerAnyTriggerAny4_0 = "TriggerAny.Trigger Any (4)"(In 3=@n29.Value, In 1=VariableAccessor.Trigger(Variable=@n57.Variable).Value, In 0=@n47.Value, In 2=@n55.Value)
   n58 = VariableAccessor.Trigger(Variable=VariableMutator.Trigger(Variable=InitVariable.Trigger().Variable, Value=@TriggerAnyTriggerAny4_0.Out).Variable)
   n59 = &MSP_LayeredTrackPlayback2(
    Stem1.Stem=^Wave.Stage2.DrumLoop
    Stem2.Stem=^Wave.Stage2.ElseLoop
    Stem2.Volume=VariableAccessor.Float(Variable=VariableMutator.Float(Variable=@n56.Variable, Value=@n46.Value).Variable).Value
    Play=@n38.Value
    Stop=@n58.Value
   )
   n60 = VariableAccessor.Float(Variable=VariableMutator.Float(Variable=InitVariable.Float().Variable, Value=@n52.Value).Variable)
   WavePlayer_2 = UE.Wave Player:Stereo(
    Wave Asset=^Wave.Stage2.Boss.ElseLoop
    Play=VariableAccessor.Trigger(Variable=@n49.Variable).Value
    Stop=VariableAccessor.Trigger(Variable=@n58.Variable).Value
   ) {also {defaults {Loop="<MetaSoundBuilderResult.SUCCEEDED: 0>"}}}
   n61 = InitVariable.Trigger()
   n62 = VariableAccessor.Trigger(Variable=VariableMutator.Trigger(Variable=InitVariable.Trigger().Variable, Value=@n36.Scheduled Trigger).Variable)
   TriggerAnyTriggerAny4_3 = "TriggerAny.Trigger Any (4)"(
    In 3=VariableAccessor.Trigger(Variable=@n31.Variable).Value
    In 2=@n54.Value
    In 1=@n48.Value
    In 0=VariableAccessor.Trigger(Variable=@n62.Variable).Value
   )
   n63 = &MSP_LayeredTrackPlayback2(
    Stem1.Stem=^Wave.Stage1.DrumLoop
    Stem2.Stem=^Wave.Stage1.ElseLoop
    Stop=VariableAccessor.Trigger(Variable=VariableMutator.Trigger(Variable=@n61.Variable, Value=@TriggerAnyTriggerAny4_3.Out).Variable).Value
    Play=@n34.Value
   )
   n64 = UE.ConversionFloatToTime(In=VariableAccessor.Float(Variable=@n10.Variable).Value)
   n65 = VariableAccessor.Trigger(Variable=@n37.Variable)
   n66 = VariableMutator.Float(
    Variable=InitVariable.Float().Variable
    Value=UE.InterpTo:Audio(
     Interp Time=UE.Divide:Time by Float(PrimaryOperand=@n64.Out) {also {defaults {AdditionalOperands=60}, version [1,1]}}.Out
     Target=Value.Value:Float(Set=^NPC.Event.PlayTransitionToStage1, Reset=@n65.Value) {also {defaults {Init Value=1}}}.Output Value
    ).Value
   )
   n67 = &MSP_SwitchOnInt_4(Compare=@n19.Scheduled Trigger, Int=VariableAccessor.Int32(Variable=@n23.Variable).Value)
   WavePlayer_0 = UE.Wave Player:Stereo(
    Wave Asset=^Wave.Stage3.BossDeath
    Play=VariableAccessor.Trigger(Variable=VariableMutator.Trigger(Value=@n67.Out 2, Variable=@n44.Variable).Variable).Value
   )
   WavePlayer_1 = UE.Wave Player:Stereo(
    Wave Asset=^Wave.Stingers.Ending
    Play=VariableAccessor.Trigger(Variable=VariableMutator.Trigger(Value=@n28.Scheduled Trigger, Variable=@n4.Variable).Variable).Value
   )
   WavePlayer_3 = UE.Wave Player:Stereo(Wave Asset=^Wave.Stingers.TransitionToStage3, Play=@n57.Value)
   WavePlayer_4 = UE.Wave Player:Stereo(Wave Asset=^Wave.Stingers.TransitionToStage2, Play=@n62.Value)
   WavePlayer_6 = UE.Wave Player:Stereo(
    Wave Asset=^Wave.Stage2.BossDeath
    Play=VariableAccessor.Trigger(Variable=VariableMutator.Trigger(Value=@n67.Out 1, Variable=@n43.Variable).Variable).Value
   )
   AudioMixerAudioMixerStereo8_0 = "AudioMixer.Audio Mixer (Stereo, 8)"(
    Gain 3=^Mixing.BossDeathStinger
    Gain 6=^Mixing.BossDeathStinger
    Gain 7=^Mixing.Ending
    Gain 0=^Mixing.Stage1
    Gain 5=^Mixing.Stage3
    Gain 1=^Mixing.TransitionToPhase2
    Gain 4=^Mixing.TransitionToPhase3
    In 5 R=UE.Add:Audio(
     PrimaryOperand=@n51.Out R
     AdditionalOperands=UE.Multiply:Audio by Float(AdditionalOperands=@n53.Value, PrimaryOperand=@WavePlayer_5.Out Right).Out
    ).Out
    In 5 L=UE.Add:Audio(
     PrimaryOperand=@n51.Out L
     AdditionalOperands=UE.Multiply:Audio by Float(AdditionalOperands=@n53.Value, PrimaryOperand=@WavePlayer_5.Out Left).Out
    ).Out
    In 2 L=UE.Add:Audio(
     PrimaryOperand=@n59.Out L
     AdditionalOperands=UE.Multiply:Audio by Float(AdditionalOperands=@n60.Value, PrimaryOperand=@WavePlayer_2.Out Left).Out
    ).Out
    In 2 R=UE.Add:Audio(
     PrimaryOperand=@n59.Out R
     AdditionalOperands=UE.Multiply:Audio by Float(AdditionalOperands=@n60.Value, PrimaryOperand=@WavePlayer_2.Out Right).Out
    ).Out
    In 0 L=@n63.Out L
    In 0 R=@n63.Out R
    Gain 2=UE.Multiply:Float(PrimaryOperand=^Mixing.Stage2, AdditionalOperands=VariableAccessor.Float(Variable=@n66.Variable).Value).Out
    In 6 L=@WavePlayer_0.Out Left
    In 6 R=@WavePlayer_0.Out Right
    In 7 L=@WavePlayer_1.Out Left
    In 7 R=@WavePlayer_1.Out Right
    In 4 L=@WavePlayer_3.Out Left
    In 4 R=@WavePlayer_3.Out Right
    In 1 L=@WavePlayer_4.Out Left
    In 1 R=@WavePlayer_4.Out Right
    In 3 L=@WavePlayer_6.Out Left
    In 3 R=@WavePlayer_6.Out Right
   )
   Multiply_3 = UE.Multiply:Audio by Float(AdditionalOperands=^MasterVolume, PrimaryOperand=@AudioMixerAudioMixerStereo8_0.Out R)
   Multiply_5 = UE.Multiply:Audio by Float(AdditionalOperands=^MasterVolume, PrimaryOperand=@AudioMixerAudioMixerStereo8_0.Out L)
   TriggerCompare.Clamp:Bool(A=^bIsBossGoingToSpawn) {also {defaults {B="<MetaSoundBuilderResult.SUCCEEDED: 0>"}}}
   TriggerCompare.Clamp:Bool(A=^bIsBossGoingToSpawn) {also {defaults {B="<MetaSoundBuilderResult.SUCCEEDED: 0>"}}}
   VariableMutator.Trigger(Variable=InitVariable.Trigger().Variable, Value=@TriggerCounter_0.On Trigger)
  ]
  links [
   {dst="@UE.OutputFormat.Stereo.Audio:1", src=Multiply_3.Out}
   {dst="@UE.OutputFormat.Stereo.Audio:0", src=Multiply_5.Out}
   {dst="@CurrentPhase", src=TriggerRouteTriggerRouteInt324_0.Value}
   {dst="@UE.Source.OneShot.OnFinished", src=WavePlayer_1.On Finished}
  ]
 }]
properties {
 assetImportData=&self:AssetImportData
 attenuationSettings=&/AS_NPC_Music_Template
 maxDistance=16000
 numChannels=2
 outputFormat=Stereo
 sampleRate=48000
 soundClassObject=&Fort_Music_NPC
 virtualizationMode=PlayWhenSilent
 soundSubmixSends [{}]
}
