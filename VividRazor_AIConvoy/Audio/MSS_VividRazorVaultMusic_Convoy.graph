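# Stereo MetaSound preset of MSS_VividRazorVaultMusic that supplies staged NPC music for the VividRazor AI Convoy.
# Inherits the parent's trigger-driven interface for three loop stages, transitions, boss events and an ending.
# Overrides waves with BreakIn percussion/else loops, transition stingers and Wolf death stingers; boss loop slots reuse the regular Stage 2 and 3 Else loops.
# Uses convoy-specific attenuation, the NPC music sound class and play-when-silent virtualization.
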MetaSoundSource /VividRazor_AIConvoy/Audio/MSS_VividRazorVaultMusic_Convoy
preset_parent=/FortNPCMusic/Audio/MSS_VividRazorVaultMusic
inputs [
 bIsBossGoingToSpawn {type=Bool}
 bIsBossPresent {type=Bool}
 EndingDelay {type=Int32}
 MasterVolume {type=Float}
 Mixing.BossDeathStinger {type=Float}
 Mixing.Ending {type=Float}
 Mixing.Stage1 {type=Float}
 Mixing.Stage2 {type=Float}
 Mixing.Stage3 {type=Float}
 Mixing.TransitionToPhase2 {type=Float}
 Mixing.TransitionToPhase3 {type=Float}
 NPC.Event.BossAppeared {type=Trigger}
 NPC.Event.BossEliminated {type=Trigger}
 NPC.Event.PlayEnding {type=Trigger}
 NPC.Event.PlayStage1Loop {type=Trigger}
 NPC.Event.PlayStage2Loop {type=Trigger}
 NPC.Event.PlayStage3Loop {type=Trigger}
 NPC.Event.PlayTransitionToStage1 {type=Trigger}
 NPC.Event.PlayTransitionToStage2 {type=Trigger}
 NPC.Event.PlayTransitionToStage3 {type=Trigger}
 TransitionLengthsBars {type=Int32:Array}
 UE.Source.OnPlay {type=Trigger}
 Wave.Stage1.DrumLoop {type=WaveAsset}
 Wave.Stage1.ElseLoop {type=WaveAsset}
 Wave.Stage2.Boss.ElseLoop {type=WaveAsset}
 Wave.Stage2.BossDeath {type=WaveAsset}
 Wave.Stage2.DrumLoop {type=WaveAsset}
 Wave.Stage2.ElseLoop {type=WaveAsset}
 Wave.Stage3.Boss.ElseLoop {type=WaveAsset}
 Wave.Stage3.BossDeath {type=WaveAsset}
 Wave.Stage3.DrumLoop {type=WaveAsset}
 Wave.Stage3.ElseLoop {type=WaveAsset}
 Wave.Stingers.Ending {type=WaveAsset}
 Wave.Stingers.TransitionToStage2 {type=WaveAsset}
 Wave.Stingers.TransitionToStage3 {type=WaveAsset}
]
interfaces [UE.OutputFormat.Stereo:1.0, UE.Source.OneShot:1.0, UE.Source:1.1]
outputs [
 CurrentPhase {type=Int32}
 "UE.OutputFormat.Stereo.Audio:0" {type=Audio}
 "UE.OutputFormat.Stereo.Audio:1" {type=Audio}
 UE.Source.OneShot.OnFinished {type=Trigger}
]
pages [Default {input_defaults {
   EndingDelay=3
   MasterVolume=1
   Mixing.BossDeathStinger=1.4
   Mixing.Ending=1
   Mixing.Stage1=1
   Mixing.Stage2=1
   Mixing.Stage3=1
   Mixing.TransitionToPhase2=1
   Mixing.TransitionToPhase3=1
   NPC.Event.BossAppeared=null
   NPC.Event.BossEliminated=null
   NPC.Event.PlayEnding=null
   NPC.Event.PlayStage1Loop=null
   NPC.Event.PlayStage2Loop=null
   NPC.Event.PlayStage3Loop=null
   NPC.Event.PlayTransitionToStage1=null
   NPC.Event.PlayTransitionToStage2=null
   NPC.Event.PlayTransitionToStage3=null
   UE.Source.OnPlay=null
   Wave.Stage1.DrumLoop=/VividRazor_AIConvoy/Audio/wav/MX_S34_VividRazor_02_BreakIn_Stage_01_LP_Perc
   Wave.Stage1.ElseLoop=/VividRazor_AIConvoy/Audio/wav/MX_S34_VividRazor_02_BreakIn_Stage_01_LP_Else
   Wave.Stage2.Boss.ElseLoop=/VividRazor_AIConvoy/Audio/wav/MX_S34_VividRazor_02_BreakIn_Stage_02_LP_Else
   Wave.Stage2.BossDeath=/VividRazor_AIConvoy/Audio/wav/MX_S34_VividRazor_02_BreakIn_Wolf_Stage_02_Death_STNG
   Wave.Stage2.DrumLoop=/VividRazor_AIConvoy/Audio/wav/MX_S34_VividRazor_02_BreakIn_Stage_02_LP_Perc
   Wave.Stage2.ElseLoop=&MX_S34_VividRazor_02_BreakIn_Stage_02_LP_Else
   Wave.Stage3.Boss.ElseLoop=/VividRazor_AIConvoy/Audio/wav/MX_S34_VividRazor_02_BreakIn_Stage_03_LP_Else
   Wave.Stage3.BossDeath=/VividRazor_AIConvoy/Audio/wav/MX_S34_VividRazor_02_BreakIn_Wolf_Stage_03_Death_STNG
   Wave.Stage3.DrumLoop=/VividRazor_AIConvoy/Audio/wav/MX_S34_VividRazor_02_BreakIn_Stage_03_LP_Perc
   Wave.Stage3.ElseLoop=&MX_S34_VividRazor_02_BreakIn_Stage_03_LP_Else
   Wave.Stingers.Ending=/VividRazor_AIConvoy/Audio/wav/MX_S34_VividRazor_02_BreakIn_Stage_04_STNG
   Wave.Stingers.TransitionToStage2=/VividRazor_AIConvoy/Audio/wav/MX_S34_VividRazor_02_BreakIn_Stage_01_STNG
   Wave.Stingers.TransitionToStage3=/VividRazor_AIConvoy/Audio/wav/MX_S34_VividRazor_02_BreakIn_Stage_02_STNG
   bIsBossGoingToSpawn=null
   bIsBossPresent=null
   TransitionLengthsBars [4,4]
  }}]
properties {
 assetImportData=&self:AssetImportData
 attenuationSettings=&/ATT_VividRazorVaultMusic_Convoy
 maxDistance=16000
 numChannels=2
 outputFormat=Stereo
 sampleRate=48000
 soundClassObject=&Fort_Music_NPC
 virtualizationMode=PlayWhenSilent
 soundSubmixSends [{}]
}
