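; Server-side spawn point provider for AI convoy followers, linking what it spawns to the instigating convoy driver.
; On BeginPlay it caches the instigator controller's patrol path and binds to OnActorSpawned, logging invalid setups.
; Spawned NPC bots receive that path and the instigator's AI group tags, and are counted toward a limit that destroys the provider.
; Spawned vehicles, detected when the bot-controller cast fails, are passed to the driver's TrackFollowerVehicles.
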(event EventBeginPlay
  (bind _self self)
  (Networking|SwitchHasAuthority
    (:Authority
      (bind _asbp_phoebe_controller_non_participant (Utilities|Casting|CastToBP_PhoebeController_NonParticipant (Game|GetInstigatorController))
        (:then
          (Utilities|IsValid (Actor|GetComponentbyClass _asbp_phoebe_controller_non_participant "/Script/FortniteAI.FortAthenaNpcPatrollingComponent")
            (:Is%20Valid
              (Utilities|IsValid (Class|FortAthenaNpcPatrollingComponent|GetPatrolPath (Actor|GetComponentbyClass _asbp_phoebe_controller_non_participant "/Script/FortniteAI.FortAthenaNpcPatrollingComponent") :!class /Script/FortniteAI.FortAthenaNpcPatrollingComponent)
                (:Is%20Valid
                  (Variables|Default|SetInstigatorPatrolPath (Class|FortAthenaNpcPatrollingComponent|GetPatrolPath (Actor|GetComponentbyClass _asbp_phoebe_controller_non_participant "/Script/FortniteAI.FortAthenaNpcPatrollingComponent") :!class /Script/FortniteAI.FortAthenaNpcPatrollingComponent))
                  (EventDispatchers|BindEventtoOnActorSpawned (delegate Custom|ActorSpawned)))
                (:Is%20Not%20Valid
                  (Fortnite|FortShippingLog _self ":: Instigator Patrolling Component does not have a valid Path" true))))
            (:Is%20Not%20Valid
              (Fortnite|FortShippingLog _self ":: Instigator Controller does not contain a Patrolling Component!" true)))
          (Actor|DestroyActor))
        (:CastFailed
          (Fortnite|FortShippingLog _self ":: Provider was spawned with an Invalid Controller." true))))))

(event Custom|ActorSpawned
  (params (SpawnedActor :type Actor%20Object%20Reference))
  (bind _spawnedactor (Variables|Default|GetSpawnedActor))
  (bind _returnvalue (Actor|GetComponentbyClass _spawnedactor "/Script/FortniteAI.FortPawnComponent_AIGroup"))
  (bind _returnvalue_2 (Game|GetInstigator))
  (bind _returnvalue_3 (Actor|GetComponentbyClass _returnvalue_2 "/Script/FortniteAI.FortPawnComponent_AIGroup"))
  (Variables|Default|SetSpawnedActor SpawnedActor)
  (bind _asbp_bot_controller_npc_base (Utilities|Casting|CastToBP_BotController_NPC_Base (Fortnite|GetControllerfromActor SpawnedActor))
    (:then
      (Utilities|IsValid (Actor|GetComponentbyClass _asbp_bot_controller_npc_base "/Script/FortniteAI.FortAthenaNpcPatrollingComponent")
        (:Is%20Valid
          (Class|FortAthenaNpcPatrollingComponent|SetPatrolPath (Actor|GetComponentbyClass _asbp_bot_controller_npc_base "/Script/FortniteAI.FortAthenaNpcPatrollingComponent") (Variables|Default|GetInstigatorPatrolPath) true)
          (Utilities|IsValid _returnvalue_2
            (:Is%20Valid
              (Utilities|IsValid _returnvalue_3
                (:Is%20Valid
                  (Utilities|IsValid _returnvalue
                    (:Is%20Valid
                      (AIGroup|AddtoGroupbyTags _returnvalue (AIGroup|GetGroupTags _returnvalue_3)))
                    (:Is%20Not%20Valid
                      (Development|PrintString "ERROR:: MOONFLAXAI:: INVALID AI GROUP COMPONENT")))
                  (bind _result (Math|Integer|IncrementInt (Variables|Default|GetSpawnAttemps))
                    (:%20%20
                      (if (>= _result (Variables|Default|GetMaxAllowedSpawnedFromProvider))
                        (Actor|DestroyActor)))))
                (:Is%20Not%20Valid
                  _)))
            (:Is%20Not%20Valid
              (Development|PrintString "ERROR:: MOONFLAXAI:: INVALID INSTIGATOR"))))
        (:Is%20Not%20Valid
          (Development|PrintString "ERROR:: MOONFLAXAI:: INVALID PATROLLING COMPONENT"))))
    (:CastFailed
      (Utilities|Casting|CastToFortAthenaVehicle (Variables|Default|GetSpawnedActor)
        (:then
          (bind _asbp_player_pawn_vivid_razor_convoy_driver (Utilities|Casting|CastToBP_PlayerPawn_VividRazor_ConvoyDriver (Game|GetInstigator)))
          (Class|BPPlayerPawnVividRazorConvoyDriver|TrackFollowerVehicles _asbp_player_pawn_vivid_razor_convoy_driver (Variables|Default|GetSpawnedActor)))
        (:CastFailed
          _)))))

(fn ConstructionScript)
